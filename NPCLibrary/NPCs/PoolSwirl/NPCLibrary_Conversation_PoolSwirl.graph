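# Conversation database for the PoolSwirl NPC, holding its greeting and the branches it offers players.
# A single entry point leads to a Speech node whose default line can be replaced by prioritized contextual lines for first meetings, match conversation count or a specific outfit.
# That node fans out through link nodes to remote entry tags for NPC quests, basic services, PoolSwirl dialogue and Season 33 weapon expertise tutorial and shotgun completion conversations.
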ConversationDatabase /NPCLibrary/NPCs/PoolSwirl/NPCLibrary_Conversation_PoolSwirl
entry_points [{tag=Conversation.Entry.PoolSwirl, destinations [#1]}]
nodes [
 ConversationEntryPointNode_7 {
  id=#1
  class=ConversationEntryPointNode
  outputs [#2]
  properties {entryTag {tagName=Conversation.Entry.PoolSwirl}}
 }
 Speech_C_4 {
  id=#2
  class=Speech_C
  outputs [
   #3
   #4
   #5
   #6
   #7
  ]
  properties {
   nodeName=""
   generalConfig {
    defaultMessage=May the seas always carry you home.
    contextualMessages [
     {
      message=May the sea dragon keep you.
      contextRequirements [{requirement {class=FortControllerRequirement_IsFirstEverConversationWithNPC}}]
      priority {curve {rowName=Tandem.Speech.Priority.FirstEverInteraction}}
     }
     {
      message=May the seas always carry you home.
      contextRequirements [{requirement {class=FortControllerRequirement_CheckNPCMatchConversationCount, properties {nPCUniqueIdTag {tagName=AISpawnerData.Type.Tandem.PoolSwirl}}}}]
      priority {curve {rowName=None}}
     }
     {
      message="May you be blessed by the sun, the moon and the waves."
      contextRequirements [{requirement {class=FortControllerRequirement_CheckNPCMatchConversationCount, properties {nPCUniqueIdTag {tagName=AISpawnerData.Type.Tandem.PoolSwirl}}}}]
      priority {curve {rowName=None}}
     }
     {
      message=I see the restless warrior in you. You will find your purpose.
      contextRequirements [{requirement {class=FortControllerRequirement_HasCID, properties {allowedCIds [&Character_CubicVice_Dual]}}}]
      priority {curve {rowName=Tandem.Speech.Priority.OutfitReaction}}
     }
    ]
   }
  }
 }
 ConversationLinkNode_4 {id=#7, class=ConversationLinkNode, properties {remoteEntryTag {tagName=Conversation.Entry.PoolSwirl.Dialogue}}}
 ConversationLinkNode_3 {
  id=#6
  class=ConversationLinkNode
  properties {remoteEntryTag {tagName=Conversation.Entry.Season33.Weaponexpertise.Tutorial.01}}
 }
 ConversationLinkNode_2 {
  id=#5
  class=ConversationLinkNode
  properties {remoteEntryTag {tagName=Conversation.Entry.Season33.Weaponexpertise.Shotgun.Complete.01}}
 }
 ConversationLinkNode_0 {id=#4, class=ConversationLinkNode, properties {remoteEntryTag {tagName=Conversation.Entry.BasicServices}}}
 ConversationLinkNode_1 {id=#3, class=ConversationLinkNode, properties {remoteEntryTag {tagName=Conversation.Entry.NPCLibrary.Quests}}}
]
