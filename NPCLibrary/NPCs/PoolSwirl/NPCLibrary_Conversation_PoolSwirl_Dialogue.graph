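# Conversation database for the Pool Swirl NPC's dialogue, entered by its dialogue entry tag.
# An opening question node offers choices for Sea Dragon, Henry The Turtle, or Back.
# Each topic plays a linear chain of speech lines about the dragon's heart and oaths or the turtle's rescue and lost love.
# Both topic chains converge on a shared return-to-start node shown as a BACK choice.
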ConversationDatabase /NPCLibrary/NPCs/PoolSwirl/NPCLibrary_Conversation_PoolSwirl_Dialogue
entry_points [{tag=Conversation.Entry.PoolSwirl.Dialogue, destinations [#1]}]
nodes [
 ConversationEntryPointNode_0 {
  id=#1
  class=ConversationEntryPointNode
  outputs [#2]
  properties {entryTag {tagName=Conversation.Entry.PoolSwirl.Dialogue}}
 }
 Speech_C_0 {
  id=#2
  class=Speech_C
  outputs [#3, #4, #5]
  properties {nodeName="", softTaskIcon=&T_UI_NPC_question, generalConfig {defaultMessage=What would you like to know?}}
  sub_nodes [ConversationChoiceNode_5 {class=ConversationChoiceNode, properties {defaultChoiceDisplayText=Ask about...}}]
 }
 Speech_C_2 {
  id=#5
  class=Speech_C
  outputs [#6]
  properties {
   nodeName=""
   softTaskIcon=&T_UI_NPC_question
   generalConfig {defaultMessage="We have banished the dark thorns that menaced The Great Turtle and located his lost love!"}
  }
  sub_nodes [ConversationChoiceNode_15 {class=ConversationChoiceNode, properties {defaultChoiceDisplayText=Henry The Turtle}}]
 }
 Speech_C_3 {
  id=#6
  class=Speech_C
  outputs [#7]
  properties {
   nodeName=""
   softTaskIcon=&T_UI_NPC_question
   generalConfig {defaultMessage=A twofold happy ending. I'm honored to have been a part of it.}
  }
 }
 ReturnToStart_C_23 {
  id=#7
  class=ReturnToStart_C
  properties {softTaskIcon=&T_UI_NPC_Back}
  sub_nodes [ConversationChoiceNode_32 {class=ConversationChoiceNode, properties {defaultChoiceDisplayText=BACK}}]
 }
 Back_C_0 {
  id=#4
  class=Back_C
  sub_nodes [ConversationChoiceNode_7 {class=ConversationChoiceNode, properties {defaultChoiceDisplayText=BACK}}]
 }
 Speech_C_7 {
  id=#3
  class=Speech_C
  outputs [#8]
  properties {
   nodeName=""
   softTaskIcon=&T_UI_NPC_question
   generalConfig {defaultMessage=I carry the heart of the lost sea dragon within me.}
  }
  sub_nodes [ConversationChoiceNode_0 {class=ConversationChoiceNode, properties {defaultChoiceDisplayText=Sea Dragon}}]
 }
 Speech_C_15 {
  id=#8
  class=Speech_C
  outputs [#9]
  properties {
   nodeName=""
   softTaskIcon=&T_UI_NPC_question
   generalConfig {defaultMessage=What does that mean? Honestly... I am not totally sure.}
  }
 }
 Speech_C_14 {
  id=#9
  class=Speech_C
  outputs [#10]
  properties {
   nodeName=""
   softTaskIcon=&T_UI_NPC_question
   generalConfig {defaultMessage="I am duty-bound to carry out the dragon's oaths, and search for its lost bones."}
  }
 }
 Speech_C_1 {
  id=#10
  class=Speech_C
  outputs [#7]
  properties {
   nodeName=""
   softTaskIcon=&T_UI_NPC_question
   generalConfig {defaultMessage="I am sworn to it, and that purpose gives me comfort."}
  }
 }
]
